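; Blueprint for a mid-section large industrial pipe prop variant in the Helios environment, with an optional light fixture.
; The construction script runs the parent setup, then, if the fixture is valid, applies the mesh color to it as a secondary mesh.
; It creates a dynamic material instance on the fixture's first material slot and sets the emissive tint from the emissive color variable.
; The LoadMesh custom event has no logic here.
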(event Custom|LoadMesh)

(fn ConstructionScript
  (bind _lightfixture (Variables|Default|GetLightFixture))
  (|CallParentFunction)
  (if (Utilities|IsValid (Variables|Default|GetLightFixture) :!class /Script/Engine.KismetSystemLibrary)
    (Default|SetColoronSecondaryMesh (Variables|Default|GetMeshColor) _lightfixture)
    (bind _returnvalue (Rendering|Material|CreateDynamicMaterialInstance _lightfixture 0 (Rendering|Material|GetMaterial _lightfixture 0) "None" :!class /Script/Engine.PrimitiveComponent))
    (Rendering|Material|SetVectorParameterValue _returnvalue "SFX_Emissive_Tint(Expand)" (Variables|Default|GetEmissiveColor) :!class /Script/Engine.MaterialInstanceDynamic)))
