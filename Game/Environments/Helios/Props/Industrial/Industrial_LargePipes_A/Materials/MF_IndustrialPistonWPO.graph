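# Material function for an industrial piston that adds animated World Position Offset and emissive glow to incoming material attributes.
# A sine-shaped cycle runs from GameTime and LoopDuration inputs, with a custom primitive data halt time blending toward a world-position phase offset and a static switch for free-running preview.
# Pre-skinned position gradients build rod and head masks.
# Offset cycle phase, vertex normal and contrast drive a pulsing two-color emissive.
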MaterialFunction /Game/Environments/Helios/Props/Industrial/Industrial_LargePipes_A/Materials/MF_IndustrialPistonWPO
graph {nodes [
  AppendVector(A=Constant2Vector(), B=Add(A=NamedRerouteUsage()) {constB=-0.05})
  ComponentMask(None=TextureCoordinate()) {r=true}
  n1 = GetMaterialAttributes(None=FunctionInput(Preview=NamedRerouteUsage()) {bUsePreviewValueAsDefault=true, inputName=MA, inputType=FunctionInput_MaterialAttributes}) {attributeGetTypes [#1]}
  FunctionOutput(None=SetMaterialAttributes!(MaterialAttributes=@n1.MaterialAttributes, WorldPositionOffset=NamedRerouteUsage(), EmissiveColor=NamedRerouteUsage()) {attributeSetTypes [#2, #3]})
  n2 = Reroute()
  n3 = Constant() {r=0.05}
  n4 = Multiply(A=Sine(None=@n2), B=@n3)
  MaterialFunctionCall!(X range Y range, Gradient Height, ShowGrid?, Thickness, Color, "f(x)"=Sine(None=Add(A=@n2, B=@n4))) {materialFunction=/Engine/Functions/Engine_MaterialFunctions01/Debug/PlotFunctionOnGraph}
  n5 = TextureSample() {samplerType=SAMPLERTYPE_Masks, texture=/Game/Environments/Helios/Sets/ScrapHeap/Textures/T_ScrapHeap_Piston_Cage_A_S}
  Multiply(A=@n5.B) {constB=0.2}
  n6 = NamedRerouteUsage()
  n7 = Time()
  n8 = ScalarParameter() {defaultValue=3, group=Piston WPO, parameterName=Preview_TimeLoop}
  n9 = ObjectPositionWS()
  n10 = ComponentMask(None=@n9) {r=true}
  n11 = ComponentMask(None=@n9) {g=true}
  n12 = Frac(None=Divide(A=Add(A=@n10, B=@n11), B=ScalarParameter() {defaultValue=2000, group=Piston WPO, parameterName=Anim Delay World Pos}) {constB=10})
  n13 = Time()
  n14 = NamedRerouteDeclaration(None=FunctionInput(Preview=Constant()) {inputName=GameTime, inputType=FunctionInput_Scalar, sortPriority=1}) {name=Game Time, nodeColor=<rgba 0.7578 0 1 1>}
  n15 = Subtract(A=@n13, B=@n14)
  n16 = FunctionInput(Preview=ScalarParameter() {defaultValue=8, group=Piston WPO, parameterName="Loop Duration (seconds)"}) {inputName=LoopDuration, inputType=FunctionInput_Scalar, sortPriority=1}
  n17 = Saturate(None=Divide(A=Abs(None=@n15), B=NamedRerouteDeclaration(None=@n16) {name=Loop Duration, nodeColor=<rgba 0 1 0.5547 1>}) {constB=0.5})
  n18 = Saturate(None=@n14)
  n19 = NamedRerouteUsage()
  n20 = Abs(None=Subtract(A=@n13, B=@n19))
  n21 = Max(A=ScalarParameter() {group=Piston WPO, parameterName=Halt_Duration}) {constB=0.001}
  n22 = Multiply(A=Saturate(None=Divide(A=@n20, B=@n21) {constB=0.5}), B=Saturate(None=@n19))
  n23 = StaticSwitch(
   True=Add(A=Divide(A=@n7, B=@n8), B=@n12)
   False=LinearInterpolate(A=Multiply(A=@n17, B=@n18), B=@n12, Alpha=NamedRerouteDeclaration(None=@n22) {name=HaltPistonValue, nodeColor=<rgba 0.8516 1 0 1>})
   Value=NamedRerouteUsage()
  )
  n24 = Add(A=@n23, B=Multiply(A=Sine(None=@n23), B=@n3))
  n25 = Multiply(A=Add(A=Sine(None=@n24))) {constB=0.5}
  n26 = Add(A=NamedRerouteUsage(), B=Constant() {r=-0.061})
  n27 = NamedRerouteUsage()
  n28 = LinearInterpolate(
   A=AppendVector(A=Constant2Vector(), B=NamedRerouteDeclaration(None=@n25) {name=Time, nodeColor=<rgba 0 1 0.8125 1>})
   B=AppendVector(A=Constant2Vector(), B=@n26)
   Alpha=Step(Y=Subtract(A=@n26, B=@n27)) {constX=0}
  )
  NamedRerouteDeclaration(None=Multiply(A=LinearInterpolate(A=@n28, Alpha=@n6) {constB=0}, B=ScalarParameter() {defaultValue=-1400, group=Piston WPO, parameterName=Movement Range}) {constB=50}) {name=WPO_OUT, nodeColor=<rgba 0 1 0.5313 1>}
  n29 = NamedRerouteUsage()
  n30 = Multiply(
   A=Add(A=@n29, B=ScalarParameter() {defaultValue=-1, desc=-0.5, group=Piston WPO, parameterName=Emissive_Noise_Bias}) {constB=-0.5}
   B=ScalarParameter() {defaultValue=-1, desc=-2, group=Piston WPO, parameterName=Emissive_Noise_Scale}
  ) {constB=-2}
  n31 = Add(A=@n23, B=ScalarParameter() {defaultValue=0.25, group=Piston WPO, parameterName=Emissive_ClimaxTime Offset}) {constB=0.1, desc=offsets sine peak}
  n32 = Clamp(None=Frac(None=Multiply(A=@n31) {constB=-1}), Min=ScalarParameter() {defaultValue=0.2, group=Piston WPO, parameterName=Emissive_MinValue}) {minDefault=0.2}
  n33 = NamedRerouteUsage()
  n34 = Abs(None=ComponentMask(None=VertexNormalWS()) {b=true})
  n35 = Constant() {r=0.5}
  n36 = Multiply(A=@n33, B=Multiply(A=OneMinus(None=@n34), B=@n35) {constB=0.3})
  n37 = Multiply(A=@n32, B=Multiply(A=Add(A=@n33, B=@n36), B=OneMinus(None=@n35)) {constB=0.7})
  n38 = Saturate(None=Add(A=@n37, B=Subtract(A=@n37, B=@n30)))
  n39 = LinearInterpolate(
   A=VectorParameter() {group=Piston WPO, parameterName=Piston Emissive Color2, defaultValue=<rgba 0.1927 0 0.0221 1>}.RGB
   B=VectorParameter() {group=Piston WPO, parameterName=Piston Emissive Color, defaultValue=<rgba 3 0.25 0 1>}.RGB
   Alpha=MaterialFunctionCall(In=@n38, Contrast=ScalarParameter() {defaultValue=1, group=Piston WPO, parameterName=Emissive_ColorContrast}) {materialFunction=&CheapContrast}.Result
  )
  NamedRerouteDeclaration(None=LinearInterpolate(A=Multiply(A=@n39, B=@n38), Alpha=NamedRerouteUsage()) {constB=0}) {name=Emissive_OUT, nodeColor=<rgba 0 0.4375 1 1>}
  NamedRerouteDeclaration(None=ScalarParameter() {bUseCustomPrimitiveData=true, group=Piston WPO, parameterName=HaltPistonTime, primitiveDataIndex=2}) {name=HaltPiston_Time, nodeColor=<rgba 0.5728 0 1 1>}
  n40 = SetMaterialAttributes!(
   MaterialAttributes
   BaseColor=TextureSample() {texture=/Game/Environments/Helios/Sets/ScrapHeap/Textures/T_ScrapHeap_Piston_Cage_A_D}.RGB
   Specular=@n5.R
   Metallic=@n5.R
   Roughness=@n5.R
   Normal=TextureSample() {samplerType=SAMPLERTYPE_Normal, texture=/Game/Environments/Helios/Sets/ScrapHeap/Textures/T_ScrapHeap_Piston_Cage_A_N}.RGB
  ) {attributeSetTypes [
    #4
    #5
    #6
    #1
    #7
   ]}
  NamedRerouteDeclaration(None=@n40) {name=P_InputMA, nodeColor=<rgba 0.0078 0 1 1>}
  n41 = NamedRerouteUsage()
  n42 = ComponentMask(None=@n41) {b=true}
  n43 = NamedRerouteDeclaration(None=ScalarParameter() {defaultValue=80, group=Piston WPO, parameterName=TRod_Base_Height}) {name=S_TRod_BaseHeight, nodeColor=<rgba 0.7812 1 0 1>}
  n44 = NamedRerouteDeclaration(None=Step(Y=Divide(A=@n42, B=@n43) {constB=90})) {name=M_TRod_Base_Flat, nodeColor=<rgba 0 0.6719 1 1>}
  n45 = Divide(A=Abs(None=@n41), B=Constant() {r=60}) {constB=70}
  n46 = Step(Y=Max(A=ComponentMask(None=@n45) {r=true}, B=ComponentMask(None=@n45) {g=true}))
  n47 = Constant() {r=1270}
  n48 = NamedRerouteDeclaration(None=Divide(A=@n42, B=NamedRerouteDeclaration(None=@n47) {name=S_TRod_TopHeight, nodeColor=<rgba 0 0.6719 1 1>})) {name=M_TRod_Gradient, nodeColor=<rgba 0 0.6719 1 1>}
  NamedRerouteDeclaration(None=LinearInterpolate(B=Max(A=@n44, B=@n46), Alpha=Step(Y=@n48))) {name=M_TRod_Flat, nodeColor=<rgba 0 0.3203 1 1>}
  NamedRerouteDeclaration(None=StaticBool() {desc=If on Breaks WP anim in material editor preview}) {name=Preview Masks - Breaks WPO anim, nodeColor=<rgba 0 0.8125 1 1>}
  n49 = NamedRerouteUsage()
  n50 = FunctionInput(Preview=NamedRerouteUsage()) {desc="Ignore this, used to work on WPO only", inputName=Delete - Preview MF, inputType=FunctionInput_StaticBool, previewValue=<1 1 1 1>}
  n51 = StaticSwitch(True=@n49, False=NamedRerouteUsage(), Value=@n50)
  NamedRerouteDeclaration(None=@n51) {name=PreSkinPos_VI_Preview, nodeColor=<rgba 1 0.8672 0 1>}
  NamedRerouteDeclaration(None=StaticBoolParameter() {defaultValue=true, group=Piston WPO, parameterName="[DEBUG] Preview WPO anim without PIE", sortPriority=0}) {name=D_Preview wo PIE, nodeColor=<rgba 1 0 1 1>}
  n52 = Divide(A=@n42) {constB=1400}
  n53 = Abs(None=@n52)
  NamedRerouteDeclaration(None=Max(A=ComponentMask(None=@n53) {r=true}, B=ComponentMask(None=@n53) {g=true})) {name=M_PistonRG_Gradient, nodeColor=<rgba 0 0.1797 1 1>}
  NamedRerouteDeclaration(None=@n1.Roughness) {name=Roughness_IN, nodeColor=<rgba 1 0.7031 0 1>}
  n54 = Divide(A=ComponentMask(None=NamedRerouteUsage()) {b=true}) {constB=1300}
  n55 = Step(Y=@n54)
  n56 = Frac(None=Multiply(A=@n54) {constB=50})
  NamedRerouteDeclaration(None=LinearInterpolate(B=Multiply(A=@n55, B=@n56), Alpha=NamedRerouteUsage())) {name=P_DiffusePlaceholder, nodeColor=<rgba 0 0 0 1>}
  NamedRerouteDeclaration(None=StaticBoolParameter() {group=Piston WPO, parameterName=Preview WPO as Emissive}) {name=Debug_Preview WPO, nodeColor=<rgba 0.3359 0 1 1>}
  NamedRerouteUsage()
  NamedRerouteUsage()
  NamedRerouteUsage()
  NamedRerouteUsage()
  OneMinus(None=@n29)
  Reroute(None=@n32)
  n57 = NamedRerouteDeclaration(None=PreSkinnedPosition()) {name=PreSkinPos, nodeColor=<rgba 0 1 0.6484 1>}
  n58 = ComponentMask(None=NamedRerouteDeclaration(None=VertexInterpolator(VS=@n57).PS) {name=PreSkinPos_VI, nodeColor=<rgba 1 0.3281 0 1>}) {b=true}
  Saturate(None=Add(A=Divide(A=@n58) {constB=1130}) {constB=-1.2})
  StaticSwitch(True=Constant(), Value=NamedRerouteUsage())
  StaticSwitch(True=@n49, False=@n51, Value=NamedRerouteUsage())
  Step(Y=NamedRerouteDeclaration(None=@n52) {name=M_PistonH_Gradient, nodeColor=<rgba 0 0.1797 1 1>})
  Step(Y=Subtract(A=NamedRerouteUsage(), B=NamedRerouteUsage())) {constX=0.05}
  n59 = ComponentMask(None=Divide(A=NamedRerouteUsage()) {constB=1400}) {b=true}
  n60 = Multiply(
   A=Subtract(A=Multiply(A=@n59, B=@n59), B=ScalarParameter() {defaultValue=0.85, group=Piston WPO, parameterName=Emissive_PistonHeadGradient_Height})
   B=ScalarParameter() {defaultValue=1.3, group=Piston WPO, parameterName=Emissive_PistonHeadGradient_Contrast}
  ) {constB=7}
  Step(X=NamedRerouteDeclaration(None=Saturate(None=@n60)) {name=Emissive Head Mask, nodeColor=<rgba 0 1 0.0391 1>}) {constY=1}
 ]}
