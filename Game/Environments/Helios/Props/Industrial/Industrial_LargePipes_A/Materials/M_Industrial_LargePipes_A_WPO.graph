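# Nanite-enabled, layered material for Helios large industrial pipe props, adding piston-style World Position Offset animation.
# GigaMix base layers feed destructible damage, then piston WPO timed by GameTime and LoopDuration custom primitive data.
# Curie fallback, snow and ice, world-space normal conversion, DirectFX, jungle biome and RVT blending follow.
# Low-quality mobile overrides specular and roughness with constants before the Creative move-tool function.
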Material /Game/Environments/Helios/Props/Industrial/Industrial_LargePipes_A/Materials/M_Industrial_LargePipes_A_WPO
graph {
 properties {
  bAutomaticallySetUsageInEditor=false
  bUseMaterialAttributes=true
  bUsedWithNanite=true
  pixelDepthOffsetMode=PDOM_Legacy
  usedShadingModels=Default Lit
 }
 nodes [
  n1 = MaterialFunctionCall() {materialFunction=/Game/Global/Base4Layers_2022/MaterialFunctions/MF_GigaMix_2023}
  MaterialExpressionReroute_0 = Reroute(None=@n1.Result)
  n2 = ScalarParameter() {
   desc=Documentation link
   group="0 (Expand)  Documentation (RMB - Copy display name) and Debug view"
   parameterName=https://confluence.it.epicgames.com/display/FORT/Fortnite+Base+Material+2022
  }
  n3 = SetMaterialAttributes!(MaterialAttributes=@MaterialExpressionReroute_0.MaterialAttributes, World Position Offset=Multiply(A=@n2)) {attributeSetTypes [#1]}
  MaterialExpressionReroute_1 = Reroute(None=@n3)
  MaterialExpressionReroute_2 = Reroute(None=@MaterialExpressionReroute_1.Output)
  n4 = MaterialFunctionCall(Material_In=@MaterialExpressionReroute_2.MaterialAttributes) {materialFunction=&MF_FN_DestructibleDamage}
  n5 = MaterialFunctionCall!(
   In=MaterialFunctionCall!(
    Delete - Preview MF=StaticBool()
    MA=@n4.Material_Out
    GameTime=ScalarParameter() {bUseCustomPrimitiveData=true, parameterName=GameTime}
    LoopDuration=ScalarParameter() {bUseCustomPrimitiveData=true, parameterName=LoopDuration, primitiveDataIndex=1}
   ) {materialFunction=&/MF_IndustrialPistonWPO}.Result
   UseSphericalMappingOverride
  ) {materialFunction=/Game/Packages/Fortress_SharedMaterials/Base_Material_Functions/MF_CurieFallbackEffect}
  n6 = MaterialFunctionCall!(In=@n5.Result) {materialFunction=/Game/Global/Base4Layers_2022/MaterialFunctions/MF_SnowAndIceArchitecture}
  MaterialExpressionReroute_3 = Reroute(None=@n6.Result)
  Composite!(None=@MaterialExpressionReroute_3.Result) {desc=Transform to World in case we will use RVT after this point, subgraphName=Tangent to World}
  Composite!(MaterialAttributes=@n1.Result) {subgraphName=Documentation Link}
  Composite(MaterialAttributes=@MaterialExpressionReroute_1.Output) {desc=Game Shared WPO and Fire Effect, subgraphName=WPO And Shared Functions}
  MaterialExpressionReroute_12 = Reroute(None=@n12.RVT H)
  MaterialExpressionReroute_6 = Reroute(None=@MaterialExpressionReroute_3.Result)
  n7 = GetMaterialAttributes(None=@MaterialExpressionReroute_6.Input) {attributeGetTypes [#2]}
  n8 = Transform(None=@n7.Normal)
  n9 = SetMaterialAttributes!(
   MaterialAttributes=@n7.MaterialAttributes
   Normal=StaticSwitchParameter(True=MaterialProxyReplace(Realtime=@n8, MaterialProxy=@n7.Normal), False=@n8) {parameterName=Force TangentSpaceNormalsFor HLOD ProxyReplace}
  ) {attributeSetTypes [#2]}
  MaterialExpressionReroute_7 = Reroute(None=@n9)
  n10 = MaterialFunctionCall!(In=@MaterialExpressionReroute_7.Output) {materialFunction=/Game/ArtTools/Tech_Art/DirectFX/Assets/MaterialFunctions/MF_DirectFX}
  MaterialExpressionReroute_11 = Reroute(None=@n10.Result)
  n11 = MaterialFunctionCall!(AO, RVT_Height=@MaterialExpressionReroute_12."RVT_Height (S)", For Buildings=@MaterialExpressionReroute_11."For Moss (MA)", For Moss) {materialFunction=/Game/Environments/Asteria/Rocks/Materials/MaterialFunctions/MF_JungleBiome}
  MaterialExpressionReroute_13 = Reroute(None=@n11.Buildings Jungl)
  n12 = MaterialFunctionCall!(
   In=@MaterialExpressionReroute_13.Result
   "RVT Height Map (Roughness)"=GetMaterialAttributes(None=@n10.Result) {attributeGetTypes [#3]}.Roughness
  ) {materialFunction=/Game/Global/Base4Layers_2022/MaterialFunctions/MF_RVTBlending}
  Composite!(InputPin=@n12.Result) {subgraphName=Mobile Opt}
  Composite!("For Moss (MA)"=@n10.Result, "RVT_Height (S)"=@n12.RVT H) {subgraphName="Seasonal "}
  MaterialExpressionReroute_4 = Reroute(None=@n12.Result)
  n13 = GetMaterialAttributes(None=@MaterialExpressionReroute_4.InputPin)
  n14 = Constant() {r=0.5}
  n15 = Constant() {r=0.8}
  n16 = ShadingPathSwitch(
   Default=@MaterialExpressionReroute_4.InputPin
   Mobile=QualitySwitch(
    Default=@MaterialExpressionReroute_4.InputPin
    Low=SetMaterialAttributes!(MaterialAttributes=@n13.MaterialAttributes, Specular=@n14, Roughness=@n15) {attributeSetTypes [#4, #3]}
    High=@MaterialExpressionReroute_4.InputPin
    Medium=@MaterialExpressionReroute_4.InputPin
    Epic=@MaterialExpressionReroute_4.InputPin
   )
  )
  MaterialExpressionReroute_5 = Reroute(None=@n16)
  MaterialFunctionCall(MA In=@MaterialExpressionReroute_5.Output) {materialFunction=&MF_CreativeMoveTool}
  MaterialFunctionCall!(Material Attributes=MaterialFunctionCall!(In=@n4.Material_Out) {materialFunction=/Game/Athena/WinterArt/Materials/MF_WPO_WindNew}.Result) {materialFunction=/Game/Creative/Sets/MilitaryBase/Props/MF_FanSpin}
  MaterialFunctionCall!(In=@n5.Result, Custom Mask, Custom Rotation Axis) {materialFunction=/Game/Athena/WinterArt/Materials/MF_WindWPO_v2}
  PinBase!(Output=@n9) {reroutePins [Output {expression=&self:MaterialExpressionReroute_7}]}
  PinBase!() {reroutePins [Input {expression=&self:MaterialExpressionReroute_6}]}
  PinBase() {reroutePins [MaterialAttributes {expression=&self:MaterialExpressionReroute_0}]}
  PinBase!(Output=@n3) {reroutePins [Output {expression=&self:MaterialExpressionReroute_1}]}
  PinBase!() {reroutePins [MaterialAttributes {expression=&self:MaterialExpressionReroute_2}]}
  PinBase!(Result=@n6.Result) {reroutePins [Result {expression=&self:MaterialExpressionReroute_3}]}
  PinBase!() {reroutePins [InputPin {expression=&self:MaterialExpressionReroute_4}]}
  PinBase!(Output=@n16) {reroutePins [Output {expression=&self:MaterialExpressionReroute_5}]}
  PinBase!() {reroutePins [
    "For Moss (MA)" {expression=&self:MaterialExpressionReroute_11}
    "RVT_Height (S)" {expression=&self:MaterialExpressionReroute_12}
   ]}
  PinBase!(Result=@n11.Buildings Jungl) {reroutePins [Result {expression=&self:MaterialExpressionReroute_13}]}
 ]
}
